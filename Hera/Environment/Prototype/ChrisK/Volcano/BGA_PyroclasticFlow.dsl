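; Prototype volcano pyroclastic flow building actor: StartFlow sends a growing particle cloud along a spline via a timeline.
; On authority, a repeating timer box-overlaps at the cloud, with extents growing along the path, launching players upward without fall damage and destroying vulnerable props, walls, floors and roofs.
; Remotes play an explosion cue and repeatedly camera-shake nearby players until the flow finishes.
; OnRep_IsActive toggles particles and looping audio.
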(event EventBeginPlay
  (Variables|BuildingActor|SetbIsInvulnerable true)
  (Components|Timeline|SetPlayRate (Variables|BGA_PyroclasticFlow|GetFlowMovement) (Variables|Default|GetFlowPlayRate))
  (Variables|Default|SetSplineLength (Spline|GetSplineLength (Variables|Default|GetSpline)))
  (Networking|SwitchHasAuthority))

(event Custom|StartFlow
  (Utilities|FlowControl|Sequence
    (:then_0
      (Variables|Default|SetIsActive true)
      (bind (_direction _lerp) (Timeline|FlowMovement PlayFromStart)
        (:Update
          (Variables|Default|SetCurrentAlpha _lerp)
          (Transformation|SetWorldLocationAndRotation (Variables|Default|GetP_PyroclasticFlow_A) (.location (Spline|GetTransformatDistanceAlongSpline (Variables|Default|GetSpline) (* (Variables|Default|GetSplineLength) _lerp) "World" false)) (.rotation (Spline|GetTransformatDistanceAlongSpline (Variables|Default|GetSpline) (* (Variables|Default|GetSplineLength) _lerp) "World" false)) false false)
          (Transformation|SetWorldScale3D (Variables|Default|GetP_PyroclasticFlow_A) (Math|Vector|Lerp(Vector) "0.600000,0.600000,0.600000" "1.100000,1.100000,1.100000" _lerp)))
        (:Finished
          (Variables|Default|SetIsActive false)
          (Utilities|Time|ClearandInvalidateTimerbyHandle (Variables|Default|GetFlowTimerHandle))
          (CallFunction|StopCameraShakes))))
    (:then_1
      (Networking|SwitchHasAuthority
        (:Authority
          (bind _returnvalue (Utilities|Time|SetTimerbyEvent (delegate Custom|OverlapActors) 0.2 true))
          (Variables|Default|SetFlowTimerHandle _returnvalue)
          (bind (_outactors _returnvalue_2) (Collision|BoxOverlapActorsWithOrientation (Transformation|GetWorldLocation (Variables|Default|GetP_PyroclasticFlow_A)) (Math|Vector|Lerp(Vector) "400.000000,800.000000,1000.000000" "650.000000,1500.000000,1650.000000" (Variables|Default|GetCurrentAlpha)) (Transformation|GetWorldRotation (Variables|Default|GetP_PyroclasticFlow_A)) (Utilities|Array|MakeArray "ObjectTypeQuery1" "ObjectTypeQuery3") "/Script/Engine.Actor"))
          (Variables|Default|SetCurrentOverlapActors _outactors)
          (if _returnvalue_2
            (CallFunction|PyroFlowDynamicOverlapFunction _outactors)))
        (:Remote
          (Audio|Components|Audio|Play (Variables|Default|GetClinger_Expl_Cue) 0.0)
          (bind _returnvalue_3 (Utilities|Time|SetTimerbyEvent (delegate Custom|CameraShakes) 1.0 true))
          (Variables|Default|SetClientCameraShakes _returnvalue_3)
          (bind (_outactors_2 _returnvalue_4) (Collision|SphereOverlapActors (Transformation|GetWorldLocation (Variables|Default|GetP_PyroclasticFlow_A)) 6e+03 (Utilities|Array|MakeArray "ObjectTypeQuery3") "/Script/FortniteGame.FortPlayerPawnAthena"))
          (if _returnvalue_4
            (for _array_element _outactors_2
              (bind _asfort_player_pawn_athena (Utilities|Casting|CastToFortPlayerPawnAthena _array_element))
              (bind _asfort_player_controller_athena (Utilities|Casting|CastToFortPlayerControllerAthena (Pawn|GetController _asfort_player_pawn_athena)))
              (Game|Feedback|ClientStartCameraShake :self _asfort_player_controller_athena :Shake "/Game/Effects/Fort_Effects/BluePrints/CameraShakes/B_Roar_CameraShake_Long.B_Roar_CameraShake_Long_C" :Scale 1.0 :PlaySpace "CameraLocal"))))))))

(event Custom|OverlapActors
  (bind (_outactors _returnvalue) (Collision|BoxOverlapActorsWithOrientation (Transformation|GetWorldLocation (Variables|Default|GetP_PyroclasticFlow_A)) (Math|Vector|Lerp(Vector) "400.000000,800.000000,1000.000000" "650.000000,1500.000000,1650.000000" (Variables|Default|GetCurrentAlpha)) (Transformation|GetWorldRotation (Variables|Default|GetP_PyroclasticFlow_A)) (Utilities|Array|MakeArray "ObjectTypeQuery1" "ObjectTypeQuery3") "/Script/Engine.Actor"))
  (Variables|Default|SetCurrentOverlapActors _outactors)
  (if _returnvalue
    (CallFunction|PyroFlowDynamicOverlapFunction _outactors)))

(event Custom|CameraShakes
  (bind (_outactors _returnvalue) (Collision|SphereOverlapActors (Transformation|GetWorldLocation (Variables|Default|GetP_PyroclasticFlow_A)) 6e+03 (Utilities|Array|MakeArray "ObjectTypeQuery3") "/Script/FortniteGame.FortPlayerPawnAthena"))
  (if _returnvalue
    (for _array_element _outactors
      (bind _asfort_player_pawn_athena (Utilities|Casting|CastToFortPlayerPawnAthena _array_element))
      (bind _asfort_player_controller_athena (Utilities|Casting|CastToFortPlayerControllerAthena (Pawn|GetController _asfort_player_pawn_athena)))
      (Game|Feedback|ClientStartCameraShake :self _asfort_player_controller_athena :Shake "/Game/Effects/Fort_Effects/BluePrints/CameraShakes/B_Roar_CameraShake_Long.B_Roar_CameraShake_Long_C" :Scale 1.0 :PlaySpace "CameraLocal"))))

(event Custom|StopCameraShakes
  (Utilities|Time|PauseTimerbyHandle (Variables|Default|GetClientCameraShakes)))

(fn OnRep_IsActive
  (if (Variables|Default|GetIsActive)
    (Components|Activation|Activate (Variables|Default|GetP_PyroclasticFlow_A) true)
    (Audio|Components|Audio|Play (Variables|Default|GetDrunkenRocket_InAirLoop_Cue) 0.0)
    (Components|Activation|Activate (Variables|Default|GetParticleSystem) true)
    (else
      (Components|Activation|Deactivate (Variables|Default|GetP_PyroclasticFlow_A))
      (Audio|Components|Audio|Stop (Variables|Default|GetDrunkenRocket_InAirLoop_Cue))
      (Components|Activation|Deactivate (Variables|Default|GetParticleSystem)))))

(fn PyroFlowDynamicOverlapFunction
  (params (Actors :type Array%20of%20Actor%20Object%20References))
  (for _array_element Actors
    (bind _output_get (LocalVariables|SetCurrentActorLocal _array_element))
    (Utilities|IsValid _output_get
      (:Is%20Valid
        (bind _asfort_player_pawn_athena (Utilities|Casting|CastToFortPlayerPawnAthena _output_get)
          (:then
            (Character|LaunchCharacter _asfort_player_pawn_athena "0.000000,0.000000,4000.000000" false true)
            (Variables|/Script/FortniteGame.FortPawn|SetbIgnoreNextFallingDamage true _asfort_player_pawn_athena))
          (:CastFailed
            (if (Variables|/Script/Engine.Actor|GetbCanBeDamaged _output_get)
              (bind _asbuilding_prop (Utilities|Casting|CastToBuildingProp (LocalVariables|GetCurrentActorLocal))
                (:then
                  (if (not (Variables|/Script/FortniteGame.BuildingActor|GetbIsInvulnerable _asbuilding_prop))
                    (Actor|DestroyActor _asbuilding_prop)))
                (:CastFailed
                  (bind _asbuilding_wall (Utilities|Casting|CastToBuildingWall (LocalVariables|GetCurrentActorLocal))
                    (:then
                      (if (not (Variables|/Script/FortniteGame.BuildingActor|GetbIsInvulnerable _asbuilding_wall))
                        (Actor|DestroyActor _asbuilding_wall)))
                    (:CastFailed
                      (bind _asbuilding_floor (Utilities|Casting|CastToBuildingFloor (LocalVariables|GetCurrentActorLocal))
                        (:then
                          (if (not (Variables|/Script/FortniteGame.BuildingActor|GetbIsInvulnerable _asbuilding_floor))
                            (Actor|DestroyActor _asbuilding_floor)))
                        (:CastFailed
                          (bind _asbuilding_roof (Utilities|Casting|CastToBuildingRoof (LocalVariables|GetCurrentActorLocal)))
                          (if (not (Variables|/Script/FortniteGame.BuildingActor|GetbIsInvulnerable _asbuilding_roof))
                            (Actor|DestroyActor _asbuilding_roof)))))))))))))))

(fn ConstructionScript)
