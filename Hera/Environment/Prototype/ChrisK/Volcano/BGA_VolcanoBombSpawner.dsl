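; Volcano bomb spawner actor that launches lava bombs from its vent on the server, on a random-interval loop toggled by start/stop events.
; Targets come from enabled target spheres, chosen by chance between a random point around the volcano, ahead of a player inside a sphere, or within a sphere.
; A vertical line trace finds the landing point, then a suggested arc velocity is passed to the spawned bomb.
; A debug flag repeatedly fires at a fixed location.
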(event EventBeginPlay
  ; Init
  (Variables|BuildingActor|SetbIsInvulnerable true)
  (Class|FortAsyncActionCurrentPlaylistReady|PlaylistDataReadyAsync
    (:OnReady
      (Networking|SwitchHasAuthority
        (:Authority
          (Utilities|FlowControl|Delay 5.0)
          (CallFunction|GetTargettingData)
          (if (Variables|Default|GetDebugSpawnBombsAtFixedLocation) :!label _label
            (CallFunction|LaunchForcedTargetBomb (Variables|Default|GetDebugSpawnBombAtFixedLocation))
            (Utilities|FlowControl|Delay (Variables|Default|GetDebugSpawnBombAtFixedInterval))
            (goto _label)))))))

(event Custom|StartLaunchingBombs
  ; Bomb Spawning and Targetting
  (Networking|SwitchHasAuthority
    (:Authority
      (CallFunction|GetTargettingData)
      (Variables|Default|SetBombSpawningActive true)
      (Utilities|FlowControl|Delay (Math|Random|RandomFloatinRange 0.9 4.0) :!label _label)
      (if (Variables|Default|GetBombSpawningActive)
        (Utilities|FlowControl|Sequence
          (:then_0
            (CallFunction|Bomb))
          (:then_1
            (if (Variables|Default|GetBombSpawningActive)
              (goto _label))))))))

(event Custom|Bomb
  ; Bomb Spawning and Targetting
  (bind (_targetvector _validlocationfound) (CallFunction|GetTargetLocationWithinSphere (Variables|Default|GetAllTargetSphereRefs)))
  (Variables|Default|SetCurrentTargetEndLoc _targetvector)
  (if _validlocationfound
    (bind _returnvalue (Class|CreativeBlueprintLibrary|GetGravityZ 0))
    (bind _outlaunchvelocity (Game|SuggestProjectileVelocityCustomArc (Transformation|GetWorldLocation (Variables|Default|GetBillboard)) (Variables|Default|GetCurrentTargetEndLoc) _returnvalue 0.375))
    (Game|SpawnActorfromClass "/Hera/Environment/Prototype/ChrisK/Volcano/BP_LavalBomb_Parent.BP_LavalBomb_Parent_C" (Math|Transform|MakeTransform (Math|Vector|MakeVector (+ (.x (Transformation|GetWorldLocation (Variables|Default|GetBillboard))) (Math|Random|RandomFloatinRange -2e+02 2e+02)) (+ (.y (Transformation|GetWorldLocation (Variables|Default|GetBillboard))) (Math|Random|RandomFloatinRange -2e+02 2e+02)) (+ (.z (Transformation|GetWorldLocation (Variables|Default|GetBillboard))) 2.5e+02)) "0, 0, 0" (Math|Vector|MakeVector (Math|Random|RandomFloatinRange 0.9 1.1) (Math|Random|RandomFloatinRange 0.9 1.1) (Math|Random|RandomFloatinRange 0.9 1.1))) "AlwaysSpawn" "MultiplyWithRoot" 0 _outlaunchvelocity 1.0 0)
    (else
      (Development|PrintString "No Bomb Hit Location Found"))))

(event Custom|StopLaunchingBombs
  ; Bomb Spawning and Targetting
  (Variables|Default|SetBombSpawningActive false))

(event Custom|LaunchForcedTargetBomb
  (params (TargetVector :type Vector))
  ; Bomb Spawning and Targetting
  (bind (_targetvector _validlocationfound) (CallFunction|GetTargetLocationWithinSphere :UseForcedTarget true :ForcedTargetVector TargetVector))
  (Variables|Default|SetCurrentTargetEndLoc _targetvector)
  (if _validlocationfound
    (bind _returnvalue (Class|CreativeBlueprintLibrary|GetGravityZ 0))
    (bind _outlaunchvelocity (Game|SuggestProjectileVelocityCustomArc (Transformation|GetWorldLocation (Variables|Default|GetBillboard)) (Variables|Default|GetCurrentTargetEndLoc) _returnvalue 0.375))
    (Game|SpawnActorfromClass "/Hera/Environment/Prototype/ChrisK/Volcano/BP_LavalBomb_Parent.BP_LavalBomb_Parent_C" (Math|Transform|MakeTransform (Math|Vector|MakeVector (+ (.x (Transformation|GetWorldLocation (Variables|Default|GetBillboard))) (Math|Random|RandomFloatinRange -2e+02 2e+02)) (+ (.y (Transformation|GetWorldLocation (Variables|Default|GetBillboard))) (Math|Random|RandomFloatinRange -2e+02 2e+02)) (+ (.z (Transformation|GetWorldLocation (Variables|Default|GetBillboard))) 2.5e+02)) "0, 0, 0" (Math|Vector|MakeVector (Math|Random|RandomFloatinRange 0.9 1.1) (Math|Random|RandomFloatinRange 0.9 1.1) (Math|Random|RandomFloatinRange 0.9 1.1))) "AlwaysSpawn" "MultiplyWithRoot" 0 _outlaunchvelocity 1.0 0)
    (else
      (Development|PrintString "No Bomb Hit Location Found"))))

(fn GetRandomPlayerInsideTargetActor
  (params (TargetActor :type BP%20Laval%20Bomb%20Target%20Sphere%20Object%20Reference))
  (bind _playerpawninsideactor (Variables|/Hera/Environment/Prototype/ChrisK/Volcano/BP_LavalBombTargetSphere.BP_LavalBombTargetSphere_C|GetPlayerPawnInsideActor TargetActor))
  (bind _output (Utilities|Array|Get(acopy) _playerpawninsideactor))
  (Utilities|Array|Shuffle _playerpawninsideactor)
  (Utilities|IsValid _output
    (:Is%20Valid
      (return _output true))
    (:Is%20Not%20Valid
      (return 0 false))))

(fn GetTargetLocationWithinSphere
  (params (TargetSphereRefs :type Array%20of%20BP%20Laval%20Bomb%20Target%20Sphere%20Object%20References) (UseForcedTarget :type Boolean) (ForcedTargetVector :type Vector))
  (bind _randomtargetlocationradiusmax (Variables|Default|GetRandomTargetLocationRadiusMax))
  (bind _randomtargetlocationradiusmin (Variables|Default|GetRandomTargetLocationRadiusMin))
  (bind _returnvalue (Math|Random|RandomFloatinRange _randomtargetlocationradiusmin _randomtargetlocationradiusmax))
  (bind _randomtargetlocationradiusmax_2 (Variables|Default|GetRandomTargetLocationRadiusMax))
  (bind _randomtargetlocationradiusmin_2 (Variables|Default|GetRandomTargetLocationRadiusMin))
  (bind _returnvalue_2 (Math|Random|RandomFloatinRange _randomtargetlocationradiusmin_2 _randomtargetlocationradiusmax_2))
  (bind _returnvalue_3 (Math|Vector|MakeVector _returnvalue_2 _returnvalue))
  (bind _self self)
  (bind _returnvalue_4 (Transformation|GetActorLocation _self))
  (bind _output (Utilities|Array|Get(acopy) TargetSphereRefs))
  (bind _returnvalue_5 (+ _returnvalue_4 _returnvalue_3))
  (bind _selectedlocationactor (LocalVariables|GetSelectedLocationActor))
  (bind _targetradius (Variables|/Hera/Environment/Prototype/ChrisK/Volcano/BP_LavalBombTargetSphere.BP_LavalBombTargetSphere_C|GetTargetRadius _selectedlocationactor))
  (if UseForcedTarget
    (LocalVariables|SetFinalStartTargetLocLocal ForcedTargetVector)
    (bind _output_get (LocalVariables|SetRadiusPositiveLocal 5e+01))
    (bind _result (Math|Float|NegateFloat _output_get)
      (:%20%20
        (LocalVariables|SetRadiusNegativeLocal _result)))
    (else
      (Utilities|Array|Shuffle TargetSphereRefs)
      (Utilities|IsValid _output
        (:Is%20Valid
          (LocalVariables|SetSelectedLocationActor _output)
          (if (>= (Variables|Default|GetRandomTargettingChance) (Math|Random|RandomFloatinRange 0.0 1.0))
            (LocalVariables|SetSelectedRandomTargetting true)
            (elif (>= (Variables|Default|GetPlayerTargettingChance) (Math|Random|RandomFloatinRange 0.0 1.0))
              (LocalVariables|SetSelectedPlayerTargetting true)
              (if (> (Utilities|Array|Length (Variables|/Hera/Environment/Prototype/ChrisK/Volcano/BP_LavalBombTargetSphere.BP_LavalBombTargetSphere_C|GetPlayerPawnInsideActor (LocalVariables|GetSelectedLocationActor))) 0)
                (bind (_player _playervalid) (CallFunction|GetRandomPlayerInsideTargetActor (LocalVariables|GetSelectedLocationActor)))
                (if _playervalid
                  (LocalVariables|SetFinalStartTargetLocLocal (+ (Transformation|GetActorLocation _player) (* (Transformation|GetActorForwardVector _player) (Math|Conversions|ToVector(Float) (Math|Random|RandomFloatinRange 3e+02 7.5e+02)))))
                  (else
                    (LocalVariables|SetSelectedPlayerTargetting false)))
                (else
                  (LocalVariables|SetSelectedPlayerTargetting false)))))
          (if (LocalVariables|GetSelectedRandomTargetting)
            (if (< (Math|Float|Absolute(Float) (Math|Vector|VectorLength (- _returnvalue_4 _returnvalue_5))) 2e+03)
              (LocalVariables|SetFinalStartTargetLocLocal (+ (Transformation|GetActorLocation (LocalVariables|GetSelectedLocationActor)) (Math|Vector|MakeVector (Math|Random|RandomFloatinRange -2e+03 2e+03) (Math|Random|RandomFloatinRange -2e+03 2e+03)))) ; fallback to nearby a random selected target actor if the bomb would land too close to the vent
              (else
                (LocalVariables|SetFinalStartTargetLocLocal _returnvalue_5)))
            (bind _output_get_2 (LocalVariables|SetRadiusPositiveLocal 2.5e+02))
            (bind _result_2 (Math|Float|NegateFloat _output_get_2)
              (:%20%20
                (LocalVariables|SetRadiusNegativeLocal _result_2)))
            (elif (LocalVariables|GetSelectedPlayerTargetting)
              (bind _output_get_3 (LocalVariables|SetRadiusPositiveLocal 2.5e+02))
              (bind _result_3 (Math|Float|NegateFloat _output_get_3)
                (:%20%20
                  (LocalVariables|SetRadiusNegativeLocal _result_3)))
              (else
                (LocalVariables|SetFinalStartTargetLocLocal (Transformation|GetWorldLocation (Variables|/Hera/Environment/Prototype/ChrisK/Volcano/BP_LavalBombTargetSphere.BP_LavalBombTargetSphere_C|GetSphere (LocalVariables|GetSelectedLocationActor))))
                (LocalVariables|SetRadiusPositiveLocal _targetradius)
                (bind _result_4 (Math|Float|NegateFloat _targetradius)
                  (:%20%20
                    (LocalVariables|SetRadiusNegativeLocal _result_4)))))))
        (:Is%20Not%20Valid
          (return "0, 0, 0" false)))))
  (bind _outhits (Collision|MultiLineTraceForObjects :Start (+ (Math|Vector|MakeVector (+ (.x (LocalVariables|GetFinalStartTargetLocLocal)) (Math|Random|RandomFloatinRange (LocalVariables|GetRadiusNegativeLocal) (LocalVariables|GetRadiusPositiveLocal))) (+ (.y (LocalVariables|GetFinalStartTargetLocLocal)) (Math|Random|RandomFloatinRange (LocalVariables|GetRadiusNegativeLocal) (LocalVariables|GetRadiusPositiveLocal))) (.z (LocalVariables|GetFinalStartTargetLocLocal))) "0.000000,0.000000,10000.000000") :End (- (Math|Vector|MakeVector (+ (.x (LocalVariables|GetFinalStartTargetLocLocal)) (Math|Random|RandomFloatinRange (LocalVariables|GetRadiusNegativeLocal) (LocalVariables|GetRadiusPositiveLocal))) (+ (.y (LocalVariables|GetFinalStartTargetLocLocal)) (Math|Random|RandomFloatinRange (LocalVariables|GetRadiusNegativeLocal) (LocalVariables|GetRadiusPositiveLocal))) (.z (LocalVariables|GetFinalStartTargetLocLocal))) "0.000000,0.000000,20000.000000") :ObjectTypes (Utilities|Array|MakeArray "ObjectTypeQuery1" "ObjectTypeQuery13") :bTraceComplex true :DrawDebugType "None" :bIgnoreSelf true :TraceColor "(R=1.000000,G=0.046423,B=0.454312,A=1.000000)" :TraceHitColor "(R=0.000000,G=1.000000,B=0.000000,A=1.000000)" :DrawTime 4.0))
  (bind _array_element (Utilities|Array|ForEachLoopwithBreak _outhits :!label _label)
    (:LoopBody
      (LocalVariables|SetEndLocLocal (.Location (Collision|BreakHitResult _array_element)))
      (Utilities|Casting|CastToLandscapeProxy (.HitActor (Collision|BreakHitResult _array_element))
        (:then
          _)
        (:CastFailed
          (Utilities|Casting|CastToBuildingStaticGeo (.HitActor (Collision|BreakHitResult _array_element))
            (:then
              _)
            (:CastFailed
              (if (== (Utilities|GetClass (.HitActor (Collision|BreakHitResult _array_element))) "/Game/Athena/Apollo/Environments/Landscape/Water/FortWaterBodyBP.FortWaterBodyBP_C")
                _)))))
      (LocalVariables|SetValidHitLocal true)
      (goto _label Break))
    (:Completed
      (return (LocalVariables|GetEndLocLocal) (LocalVariables|GetValidHitLocal)))))

(fn GetTargettingData
  (bind _outactors (Actor|GetAllActorsOfClass "/Hera/Environment/Prototype/ChrisK/Volcano/BP_LavalBombTargetSphere.BP_LavalBombTargetSphere_C"))
  (for _array_element _outactors
    (if (Variables|/Hera/Environment/Prototype/ChrisK/Volcano/BP_LavalBombTargetSphere.BP_LavalBombTargetSphere_C|GetEnabled _array_element)
      (Utilities|Array|AddUnique (Variables|Default|GetAllTargetSphereRefs) _array_element))))

(fn ConstructionScript)
