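; Prototype volcano gameplay actor that, when enabled, runs a timed, server-driven eruption after a short delay.
; Eruption starts bombs, a pyroclastic flow, rising crater lava and repeated vent launches of pawns, with client ashfall and music.
; A one-shot timeline advances lava along a spline; timed box overlaps at its front ignite and damage buildings, while pawns behind it launch upward.
; RepNotifies drive effects, vortex collision and day-sequence blending.
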(event EventBeginPlay
  ; Init
  (Variables|Default|SetDebugLavaFlowProgress 0.0)
  (Components|Timeline|SetPlayRate (Variables|BGA_Volcano_Gameplay|GetLavaFlow) 0.75)
  (Variables|Default|SetSplineLength (Spline|GetSplineLength (Variables|Default|GetSpline)))
  (Networking|SwitchHasAuthority
    (:Authority
      (Variables|Default|SetVentActive false)
      (if (Variables|Default|GetVolcanoEnabled)
        (Utilities|FlowControl|Delay 1e+01)
        (CallFunction|VolcanoEruptionChangeState true)
        (Utilities|FlowControl|Delay (Variables|Default|GetVolcanoActiveDuration))
        (CallFunction|VolcanoEruptionChangeState false)
        (Utilities|FlowControl|Delay 5.0)))))

(event Custom|VolcanoEruptionChangeState
  (params (Active :type Boolean))
  (bind _allpyroclasticflowrefs (Variables|Default|GetAllPyroclasticFlowRefs))
  ; Eruption State
  (if Active
    (Variables|Default|SetVolcanoActive true)
    (Variables|Default|SetVentActive true)
    (Networking|SwitchHasAuthority
      (:Authority
        (Class|BGAVolcanoBombSpawner|StartLaunchingBombs (Variables|Default|GetBombSpawnerRef))
        (Components|Activation|Activate (Variables|Default|GetP_Proxy_Mine_Explosion_New) false)
        (Audio|Components|Audio|Play (Variables|Default|GetClinger_Expl_Cue) 0.0)
        (CallFunction|StartCraterLava)
        (CallFunction|VentSphereYeet)
        (CallFunction|VentYeet true)
        (if (Variables|Default|GetVolcanoActive)
          (Utilities|Array|Shuffle _allpyroclasticflowrefs)
          (Utilities|FlowControl|Delay (Math|Random|RandomFloatinRange 2.0 6.0))
          (Class|BGAPyroclasticFlow|StartFlow (Utilities|Array|Get(acopy) _allpyroclasticflowrefs))
          (if (Variables|Default|GetVolcanoActive)
            (Utilities|FlowControl|Delay (Math|Random|RandomFloatinRange 25.0 4e+01)))))
      (:Remote
        (CallFunction|PlayAsh)
        (Audio|Components|Audio|Play (Variables|Default|GetDadBro_Combat_Music_02_Cue) 0.0)
        (Utilities|FlowControl|Sequence
          (:then_0
            (bind (_direction _alpha) (Timeline|PostProccessTimeline PlayFromStart)
              (:Update
                (Variables|Default|SetBlendWeightPostProccess (Math|Float|Lerp 0.0 0.6 _alpha)))
              (:Finished
                (for _array_element (Variables|Default|GetAshProtoParticles)
                  (Variables|/Hera/Environment/Prototype/ChrisK/Volcano/BP_AshFallLocal.BP_AshFallLocal_C|SetAshFallEnabled false _array_element)))))
          (:then_1
            (Utilities|FlowControl|Delay (Variables|Default|GetVolcanoActiveDuration))
            (Audio|Components|Audio|FadeOut (Variables|Default|GetDadBro_Combat_Music_02_Cue) 3.0 0.0 "Linear" :!class /Script/Engine.AudioComponent)))))
    (else
      (Variables|Default|SetVolcanoActive false)
      (Networking|SwitchHasAuthority
        (:Authority
          (Class|BGAVolcanoBombSpawner|StopLaunchingBombs (Variables|Default|GetBombSpawnerRef))
          (CallFunction|VentYeet false)
          (Timeline|PostProccessTimeline ReverseFromEnd))))))

(event Custom|PlayAsh
  ; Eruption State
  (Utilities|FlowControl|Delay 12.0)
  (for _array_element (Variables|Default|GetAshProtoParticles)
    (Variables|/Hera/Environment/Prototype/ChrisK/Volcano/BP_AshFallLocal.BP_AshFallLocal_C|SetAshFallEnabled true _array_element)))

(event Custom|VentYeet
  (params (Active :type Boolean))
  ; Vent Yeet
  (if Active
    (bind _returnvalue (Utilities|Time|SetTimerbyEvent (delegate Custom|Yeet) 2.0 true))
    (Variables|Default|SetVentYeetHandle _returnvalue)
    (else
      (Utilities|Time|PauseTimerbyHandle (Variables|Default|GetVentYeetHandle)))))

(event Custom|Yeet
  ; Vent Yeet
  (CallFunction|VentSphereYeet))

(event Custom|StartCraterLava
  ; Crater Lava
  (Utilities|FlowControl|Sequence
    (:then_0
      (bind (_direction _alpha) (Timeline|CraterMagma PlayFromStart)
        (:Update
          (Transformation|SetWorldScale3D (Variables|Default|GetCraterLava) (Math|Vector|Lerp(Vector) "55.000000,55.000000,0.100000" "55.000000,55.000000,20.000000" _alpha)))))
    (:then_1
      (Networking|SwitchHasAuthority
        (:Authority
          (Utilities|FlowControl|Delay 3.0)
          (CallFunction|StartLavalFlow))))))

(event Custom|StartLavalFlow
  ; Lava Flow
  (Utilities|FlowControl|DoOnce false
    (:Completed
      (CallFunction|LavaFlowFrontBoxCheck)
      (bind (_direction _alpha _alphalavafront) (Timeline|LavaFlow PlayFromStart)
        (:Update
          (Variables|Default|SetCurrentFlowAlpha _alpha)
          (Rendering|Material|SetScalarParameterValueonMaterials (Variables|Default|GetVolcano_Proto_CK_Stream_Flow) "LavaProgress" (Math|Float|Lerp 0.0 1.0 _alpha))
          (Variables|Default|SetCurrentLavaFrontLocation (Spline|GetLocationatDistanceAlongSpline (Variables|Default|GetSpline) (Math|Float|Lerp 0.0 (Variables|Default|GetSplineLength) (Math|Float|Clamp(Float) (- _alphalavafront 0.03))) "World")))
        (:Finished
          (Utilities|Time|PauseTimerbyHandle (Variables|Default|GetFlowBoxHandle)))))))

(event OnComponentBeginOverlap(Volcano_Proto_CK_Stream_Flow_Collision)
  (params (OverlappedComponent :type Primitive%20Component%20Object%20Reference) (OtherActor :type Actor%20Object%20Reference) (OtherComp :type Primitive%20Component%20Object%20Reference) (OtherBodyIndex :type Integer) (bFromSweep :type Boolean) (SweepResult :type Hit%20Result%20Structure%20%28by%20ref%29))
  ; Player Lava Collision
  (Networking|SwitchHasAuthority
    (:Authority
      (bind _asfort_pawn (Utilities|Casting|CastToFortPawn OtherActor))
      (bind _inside (CallFunction|IsPlayerInsideLavaFlow _asfort_pawn))
      (if _inside
        (Character|LaunchCharacter _asfort_pawn "0.000000,0.000000,3000.000000" false true)
        (else
          (Development|PrintString "Not Inside Laval"))))))

(event Custom|LavaFlowFrontBoxCheck
  ; Lava Flow Front Box Damage Overlap
  (Networking|SwitchHasAuthority
    (:Authority
      _)
    (:Remote
      _))
  (bind _returnvalue (Utilities|Time|SetTimerbyEvent (delegate Custom|OverlapCheck) 0.5 true))
  (Variables|Default|SetFlowBoxHandle _returnvalue))

(event Custom|OverlapCheck
  ; Lava Flow Front Box Damage Overlap
  (bind _outactors (Collision|BoxOverlapActorsWithOrientation (Variables|Default|GetCurrentLavaFrontLocation) (Math|Vector|Lerp(Vector) (* (Math|Vector|MakeVector 3.5e+02 3e+03 1e+02) "1.000000,0.030000,1.000000") (Math|Vector|MakeVector 3.5e+02 3e+03 1e+02) (Variables|Default|GetCurrentFlowAlpha)) (Spline|GetRotationatDistanceAlongSpline (Variables|Default|GetSpline) (* (Variables|Default|GetCurrentFlowAlpha) (Variables|Default|GetSplineLength)) "World") (Utilities|Array|MakeArray "ObjectTypeQuery3" "ObjectTypeQuery1" "ObjectTypeQuery2") "/Script/Engine.Actor"))
  (CallFunction|OverlappingLavaFlowActors _outactors))

(fn IsPlayerInsideLavaFlow
  (params (Pawn :type Fort%20Pawn%20Object%20Reference))
  (bind _output_get (LocalVariables|SetNearestDistance (Spline|GetDistanceAlongSplineatLocation (Variables|Default|GetSpline) (Spline|FindLocationClosesttoWorldLocation (Variables|Default|GetSpline) (Transformation|GetActorLocation Pawn) "World") "World")))
  (return (<= _output_get (Spline|GetDistanceAlongSplineatLocation (Variables|Default|GetSpline) (Variables|Default|GetCurrentLavaFrontLocation) "World"))))

(fn OnRep_BlendWeightPostProccess
  (Variables|/Script/DaySequence.DaySequenceModifierComponent|SetUserBlendWeight (Variables|Default|GetBlendWeightPostProccess) (Variables|/Script/DaySequence.DaySequenceModifierVolume|GetDaySequenceModifier (Variables|Default|GetDSV_ModifierRef))))

(fn OnRep_VentActive
  (if (Variables|Default|GetVentActive)
    (Collision|SetActorEnableCollision (Variables|Default|GetVortexActor) true)
    (else
      (Collision|SetActorEnableCollision (Variables|Default|GetVortexActor) false))))

(fn OnRep_VolcanoActive
  (if (Variables|Default|GetVolcanoActive)
    (Components|Activation|Activate (Variables|Default|GetP_VolcanoEruption_A) false)
    (Components|Activation|Activate (Variables|Default|GetP_Rift_RandomLightning_02) false)
    (Rendering|SetVisibility (Variables|Default|GetPointLight) true false)
    (else
      (Components|Activation|Deactivate (Variables|Default|GetP_VolcanoEruption_A))
      (Components|Activation|Deactivate (Variables|Default|GetP_Rift_RandomLightning_02))
      (Rendering|SetVisibility (Variables|Default|GetPointLight) false false))))

(fn OverlappingLavaFlowActors
  (params (Actors :type Array%20of%20Actor%20Object%20References))
  (Networking|SwitchHasAuthority
    (:Authority
      (for _array_element Actors
        (bind _asbuilding_wall (Utilities|Casting|CastToBuildingWall _array_element)
          (:then
            (Curie|ExecuteBeginFireInteract _asbuilding_wall _asbuilding_wall _asbuilding_wall 1e+01)
            (Environment|ApplyGlobalEnvironmentDamagetoActor _asbuilding_wall (Math|Random|RandomFloatinRange 25.0 75.0)))
          (:CastFailed
            (bind _asbuilding_prop (Utilities|Casting|CastToBuildingProp _array_element)
              (:then
                (Curie|ExecuteBeginFireInteract :Instigator _asbuilding_prop :EffectCauser _asbuilding_prop :Target _asbuilding_prop :Magnitude 1e+01 :PropagationFuel -1 :LandscapeIgnitionRadius 256.0)
                (Environment|ApplyGlobalEnvironmentDamagetoActor _asbuilding_prop 15.0))
              (:CastFailed
                (bind _asbuilding_floor (Utilities|Casting|CastToBuildingFloor _array_element))
                (Curie|ExecuteBeginFireInteract :Instigator _asbuilding_floor :EffectCauser _asbuilding_floor :Target _asbuilding_floor :Magnitude 1e+01 :PropagationFuel -1 :LandscapeIgnitionRadius 256.0)
                (Environment|ApplyGlobalEnvironmentDamagetoActor _asbuilding_floor (Math|Random|RandomFloatinRange 25.0 1e+02))))))))))

(fn ConstructionScript
  (Rendering|Material|SetScalarParameterValueonMaterials (Variables|Default|GetVolcano_Proto_CK_Stream_Flow) "LavaProgress" (Variables|Default|GetDebugLavaFlowProgress)))

(fn VentSphereYeet
  (bind _outactors (Collision|SphereOverlapActors (Transformation|GetWorldLocation (Variables|Default|GetVentSphere)) (Variables|/Script/Engine.SphereComponent|GetSphereRadius (Variables|Default|GetVentSphere)) (Utilities|Array|MakeArray "ObjectTypeQuery3") "/Script/FortniteGame.FortPawn"))
  (for _array_element _outactors
    (bind _asfort_pawn (Utilities|Casting|CastToFortPawn _array_element))
    (Character|LaunchCharacter _asfort_pawn (Math|Vector|MakeVector (* (.x (Math|Vector|GetUnitDirection(Vector) (Transformation|GetWorldLocation (Variables|Default|GetVentSphere)) (Transformation|GetActorLocation _asfort_pawn))) 2.5e+03) (* (.y (Math|Vector|GetUnitDirection(Vector) (Transformation|GetWorldLocation (Variables|Default|GetVentSphere)) (Transformation|GetActorLocation _asfort_pawn))) 2.5e+03) 3e+03) true true)
    (Variables|/Script/FortniteGame.FortPawn|SetbIgnoreNextFallingDamage true _asfort_pawn)))
