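; Blueprint actor for a local volcanic ash fall effect, toggled by a replicated AshFallEnabled flag.
; BeginPlay sets the flag to false.
; Its OnRep handler activates the P_Ashfall component with reset when the flag is true and deactivates it otherwise.
; The construction script is empty, and nothing in this Blueprint sets the flag to true.
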(event EventBeginPlay
  (Variables|Default|SetAshFallEnabled false))

(fn OnRep_AshFallEnabled
  (if (Variables|Default|GetAshFallEnabled)
    (Components|Activation|Activate (Variables|Default|GetP_Ashfall) true)
    (else
      (Components|Activation|Deactivate (Variables|Default|GetP_Ashfall)))))

(fn ConstructionScript)
