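; BP_LavalBombTargetSphere is a prototype volcano trigger sphere that tracks which player pawns are inside it, apparently as lava bomb targets.
; The construction script sizes the sphere from the TargetRadius variable.
; On authority only, begin overlap casts the actor to FortPlayerPawnAthena and adds it uniquely to PlayerPawnInsideActor, and end overlap removes it.
; No bomb spawning or targeting logic appears here.
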(event OnComponentBeginOverlap(Sphere)
  (params (OverlappedComponent :type Primitive%20Component%20Object%20Reference) (OtherActor :type Actor%20Object%20Reference) (OtherComp :type Primitive%20Component%20Object%20Reference) (OtherBodyIndex :type Integer) (bFromSweep :type Boolean) (SweepResult :type Hit%20Result%20Structure%20%28by%20ref%29))
  (Networking|SwitchHasAuthority
    (:Authority
      (bind _asfort_player_pawn_athena (Utilities|Casting|CastToFortPlayerPawnAthena OtherActor))
      (Utilities|Array|AddUnique (Variables|Default|GetPlayerPawnInsideActor) _asfort_player_pawn_athena))))

(event OnComponentEndOverlap(Sphere)
  (params (OverlappedComponent :type Primitive%20Component%20Object%20Reference) (OtherActor :type Actor%20Object%20Reference) (OtherComp :type Primitive%20Component%20Object%20Reference) (OtherBodyIndex :type Integer))
  (bind _playerpawninsideactor (Variables|Default|GetPlayerPawnInsideActor))
  (Networking|SwitchHasAuthority
    (:Authority
      (bind _asfort_player_pawn_athena (Utilities|Casting|CastToFortPlayerPawnAthena OtherActor))
      (if (Utilities|Array|ContainsItem _playerpawninsideactor _asfort_player_pawn_athena)
        (Utilities|Array|RemoveItem _playerpawninsideactor _asfort_player_pawn_athena)))))

(fn ConstructionScript
  (Components|Sphere|SetSphereRadius (Variables|Default|GetSphere) (Variables|Default|GetTargetRadius) true))
