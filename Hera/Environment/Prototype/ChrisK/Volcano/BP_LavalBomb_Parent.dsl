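; Parent Blueprint for a volcano lava bomb projectile that applies a set velocity and gravity scale, then explodes on stop, bounce or capsule hit.
; On authority, Boom damages and launches nearby non-DBNO players, then destroys damageable props, walls, floors and roofs in delayed sphere-overlap passes.
; Its remote branch plays explosion effects, sound and camera shake, and stops in-air audio and trail.
; A ProjectileLanded RepNotify adjusts collision or mesh visibility.
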(event EventBeginPlay
  (Variables|/Script/Engine.MovementComponent|SetVelocity (Variables|Default|GetVelocity) (Variables|Movement|GetProjectileMovementComponent))
  (Variables|/Script/Engine.ProjectileMovementComponent|SetProjectileGravityScale (Variables|Default|GetGravity) (Variables|Movement|GetProjectileMovementComponent)))

(event OnProjectileStop(ProjectileMovementComponent)
  (params (ImpactResult :type Hit%20Result%20Structure%20%28by%20ref%29))
  (Networking|SwitchHasAuthority
    (:Authority
      (Utilities|FlowControl|DoOnce false
        (:Completed
          (CallFunction|Boom))))))

(event Custom|Boom
  (Variables|Default|SetProjectileLanded true)
  (Networking|SwitchHasAuthority
    (:Authority
      (CallFunction|PlayerImpactFunction)
      ;[0.15 0.007935 0.04551 0.5] Nativize this - should be pretty simple - (destroy props, walls, roofs and floors that can be damaged and are not invulnerable)
      (bind _outactors (Collision|SphereOverlapActors (Transformation|GetWorldLocation (Variables|Collision|GetCapsuleComponent)) (Variables|Default|GetBombRadius) (Utilities|Array|MakeArray "ObjectTypeQuery1") "/Script/FortniteGame.BuildingProp" (Utilities|Array|MakeArray self)))
      (for _array_element _outactors
        (bind _asbuilding_prop (Utilities|Casting|CastToBuildingProp _array_element))
        (if (Variables|/Script/Engine.Actor|GetbCanBeDamaged _asbuilding_prop)
          (if (not (Variables|/Script/FortniteGame.BuildingActor|GetbIsInvulnerable _asbuilding_prop))
            (Actor|DestroyActor _array_element))))
      (Utilities|FlowControl|Delay 0.2)
      (bind _outactors_2 (Collision|SphereOverlapActors (Transformation|GetWorldLocation (Variables|Collision|GetCapsuleComponent)) (Variables|Default|GetBombRadius) (Utilities|Array|MakeArray "ObjectTypeQuery1") "/Script/FortniteGame.BuildingWall" (Utilities|Array|MakeArray self)))
      (for _array_element_2 _outactors_2
        (bind _asbuilding_wall (Utilities|Casting|CastToBuildingWall _array_element_2))
        (if (Variables|/Script/Engine.Actor|GetbCanBeDamaged _asbuilding_wall)
          (if (not (Variables|/Script/FortniteGame.BuildingActor|GetbIsInvulnerable _asbuilding_wall))
            (Actor|DestroyActor _asbuilding_wall))))
      (Utilities|FlowControl|Delay 0.2)
      (bind _outactors_3 (Collision|SphereOverlapActors (Transformation|GetWorldLocation (Variables|Collision|GetCapsuleComponent)) (Variables|Default|GetBombRadius) (Utilities|Array|MakeArray "ObjectTypeQuery1") "/Script/FortniteGame.BuildingFloor" (Utilities|Array|MakeArray self)))
      (for _array_element_3 _outactors_3
        (bind _asbuilding_floor (Utilities|Casting|CastToBuildingFloor _array_element_3))
        (if (Variables|/Script/Engine.Actor|GetbCanBeDamaged _asbuilding_floor)
          (if (not (Variables|/Script/FortniteGame.BuildingActor|GetbIsInvulnerable _asbuilding_floor))
            (Actor|DestroyActor _asbuilding_floor))))
      (Utilities|FlowControl|Delay 0.2)
      (bind _outactors_4 (Collision|SphereOverlapActors (Transformation|GetWorldLocation (Variables|Collision|GetCapsuleComponent)) (Variables|Default|GetBombRadius) (Utilities|Array|MakeArray "ObjectTypeQuery1") "/Script/FortniteGame.BuildingRoof" (Utilities|Array|MakeArray self)))
      (for _array_element_4 _outactors_4
        (bind _asbuilding_roof (Utilities|Casting|CastToBuildingRoof _array_element_4))
        (if (Variables|/Script/Engine.Actor|GetbCanBeDamaged _asbuilding_roof)
          (if (not (Variables|/Script/FortniteGame.BuildingActor|GetbIsInvulnerable _asbuilding_roof))
            (Actor|DestroyActor _asbuilding_roof)))))
    (:Remote
      (Components|Activation|Activate (Variables|Default|GetNS_RocketLauncher_Guided_Explode) false)
      (Rendering|SetVisibility (Variables|Default|GetSphere) false false)
      (Components|Activation|Deactivate (Variables|Default|GetP_RocketTrail_01))
      (Audio|SpawnSoundatLocation "/Game/Sounds/Fort_Weapons/Launchers/DrunkenQuad/DrunkenRocket_Explosion_Cue.DrunkenRocket_Explosion_Cue" (Transformation|GetWorldLocation (Variables|Collision|GetCapsuleComponent)))
      (Audio|Components|Audio|Stop (Variables|Default|GetMeatball_Launcher_Rocket_InAir_Cue))
      (Audio|Components|Audio|Stop (Variables|Default|GetWeapon_LAU_Rocket_Launcher_Athena_RocketInAir_Cue))
      (for _array_element_5 (Collision|GetOverlappingActors (Variables|Default|GetSphereCollision) "/Script/FortniteGame.FortPlayerPawnAthena" :!class /Script/Engine.PrimitiveComponent)
        (bind _asfort_player_controller_athena (Utilities|Casting|CastToFortPlayerControllerAthena (Fortnite|GetControllerfromActor _array_element_5)))
        (Game|Feedback|ClientStartCameraShake :self _asfort_player_controller_athena :Shake "/Game/Effects/Fort_Effects/BluePrints/CameraShakes/B_Large_Explosion_CameraShake.B_Large_Explosion_CameraShake_C" :Scale 1.0 :PlaySpace "CameraLocal")))))

(event OnProjectileBounce(ProjectileMovementComponent)
  (params (ImpactResult :type Hit%20Result%20Structure%20%28by%20ref%29) (ImpactVelocity :type Vector%20%28by%20ref%29 :default "0, 0, 0"))
  (Networking|SwitchHasAuthority
    (:Authority
      (Utilities|FlowControl|DoOnce false
        (:Completed
          (CallFunction|Boom))))))

(event OnComponentHit(CapsuleComponent)
  (params (HitComponent :type Primitive%20Component%20Object%20Reference) (OtherActor :type Actor%20Object%20Reference) (OtherComp :type Primitive%20Component%20Object%20Reference) (NormalImpulse :type Vector :default "0, 0, 0") (Hit :type Hit%20Result%20Structure%20%28by%20ref%29))
  (Development|PrintString "Hit")
  (Networking|SwitchHasAuthority
    (:Authority
      (Utilities|FlowControl|DoOnce false
        (:Completed
          (CallFunction|Boom))))))

(fn OnRep_ProjectileLanded
  (if (Variables|Default|GetProjectileLanded)
    (Collision|SetCollisionEnabled (Variables|Collision|GetCapsuleComponent) "NoCollision")
    (Rendering|SetVisibility (Variables|Default|GetSphere) false false)))

(fn PlayerImpactFunction
  (bind (_outactors _returnvalue) (Collision|SphereOverlapActors (Transformation|GetWorldLocation (Variables|Collision|GetCapsuleComponent)) (Variables|Default|GetBombRadius) (Utilities|Array|MakeArray "ObjectTypeQuery3") "/Script/FortniteGame.FortPlayerPawnAthena" (Utilities|Array|MakeArray self)))
  (if _returnvalue
    (for _array_element _outactors
      (bind _asfort_player_pawn_athena (Utilities|Casting|CastToFortPlayerPawnAthena _array_element))
      (if (not (Health|IsDBNO _asfort_player_pawn_athena))
        (Environment|ApplyGlobalEnvironmentDamagetoActor _asfort_player_pawn_athena 5.0)
        (Character|LaunchCharacter _asfort_player_pawn_athena (Math|Vector|ClampVectorSize (+ (* (Math|Vector|MakeVector (.x (Math|Vector|NegateVector (Math|Vector|GetUnitDirection(Vector) (Transformation|GetActorLocation _asfort_player_pawn_athena) (Transformation|GetActorLocation self)))) (.y (Math|Vector|NegateVector (Math|Vector|GetUnitDirection(Vector) (Transformation|GetActorLocation _asfort_player_pawn_athena) (Transformation|GetActorLocation self)))) (Math|Float|Absolute(Float) (.z (Math|Vector|NegateVector (Math|Vector|GetUnitDirection(Vector) (Transformation|GetActorLocation _asfort_player_pawn_athena) (Transformation|GetActorLocation self)))))) "1000.000000,1000.000000,1000.000000") "0.000000,0.000000,500.000000") 1e+03 1e+03) false true)))))

(fn ConstructionScript)
