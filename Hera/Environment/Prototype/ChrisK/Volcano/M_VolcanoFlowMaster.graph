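# Translucent lit master material for a volcano prototype, apparently a lava flow whose front advances with the LavaProgress parameter.
# LavaProgress slides UV V between VStart and VEnd, and a high-power, inverted LinearGradient turns it into a sharp mask.
# That mask scales a vertex-normal offset for displacing the mesh.
# A DepthFade blends the orange Emmisive color between two intensities to vary glow near intersecting geometry.
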Material /Hera/Environment/Prototype/ChrisK/Volcano/M_VolcanoFlowMaster
graph {
 properties {bAlwaysEvaluateWorldPositionOffset=true, bAutomaticallySetUsageInEditor=false, blendMode=BLEND_Translucent, usedShadingModels=Default Lit}
 nodes [
  DepthFade() {fadeDistanceDefault=35}
  n1 = AppendVector(A=Constant(), B=ScalarParameter() {defaultValue=-0.1, parameterName=VStart})
  n2 = AppendVector(A=Constant(), B=ScalarParameter() {defaultValue=1, parameterName=VEnd})
  n3 = ScalarParameter() {parameterName=LavaProgress}
  n4 = MaterialFunctionCall(UV Channel=Add(A=TextureCoordinate(), B=LinearInterpolate(A=@n1, B=@n2, Alpha=@n3)) {constB=0}) {materialFunction=&LinearGradient}
  n5 = Clamp(None=OneMinus(None=Power(Base=@n4.VGradient) {constExponent=40}))
  Multiply(A=VertexNormalWS(), B=LinearInterpolate(A=@n5, Alpha=@n5) {constB=-550})
  n6 = Multiply(A=VectorParameter() {parameterName=Emmisive, defaultValue=<rgba 1 0.0519 0 1>}.RGB)
  Reroute(None=LinearInterpolate(A=@n6, B=Multiply(A=@n6) {constB=10}, Alpha=DepthFade() {fadeDistanceDefault=75}) {constB=5})
  VectorParameter() {parameterName=Emmisive, defaultValue=<rgba 1 0.0519 0 1>}
 ]
}
